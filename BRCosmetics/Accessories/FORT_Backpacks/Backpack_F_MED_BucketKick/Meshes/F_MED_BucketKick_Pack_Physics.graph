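# Physics asset for the BucketKick backpack cosmetic, defining bodies and joints for its dangling cloth pieces.
# Two capsule chains, a three-bone A and a five-bone B, branch from dyn_cloth_base, which is constrained to a MSTR root body.
# Tight swing and twist limits restrict every joint, with reduced swing toward the chain ends.
# Larger, unconstrained pelvis, spine and thigh capsules approximate the wearer's body as collision volumes for the cloth.
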PhysicsAsset /BRCosmetics/Accessories/FORT_Backpacks/Backpack_F_MED_BucketKick/Meshes/F_MED_BucketKick_Pack_Physics
bodies {
 dyn_cloth_A_01 [{shapeType=Capsule, length=3, radius=2, shapeName=dyn_cloth_A_01_capsule, center=<0.1 0 0>, rotation=<rot 90 0 0>}]
 dyn_cloth_A_02 [{shapeType=Capsule, length=3, radius=2, shapeName=dyn_cloth_A_02_capsule, center=<0.1 0 0>, rotation=<rot 90 0 0>}]
 dyn_cloth_A_03 [{shapeType=Capsule, length=3, radius=2, shapeName=dyn_cloth_A_03_capsule, center=<0.1 0 0>, rotation=<rot 90 0 0>}]
 dyn_cloth_B_01 [{shapeType=Capsule, length=3, radius=2, shapeName=dyn_cloth_B_01_capsule, center=<0.1 0 0>, rotation=<rot 90 0 0>}]
 dyn_cloth_B_02 [{shapeType=Capsule, length=3, radius=2, shapeName=dyn_cloth_B_02_capsule, center=<0.1 0 0>, rotation=<rot 90 0 0>}]
 dyn_cloth_B_03 [
  {shapeType=Capsule, length=3, radius=2, shapeName=dyn_cloth_B_03_capsule, center=<0.1 0 0>, rotation=<rot 90 0 0>}
  {shapeType=Capsule, length=2, radius=3, shapeName=dyn_cloth_B_03_capsule, center=<0.2503 0 0.351>}
 ]
 dyn_cloth_B_04 [
  {shapeType=Capsule, length=3, radius=2, shapeName=dyn_cloth_B_04_capsule, center=<0.2256 0 1.4606>, rotation=<rot 80 -0.0001 -0.0001>}
  {shapeType=Capsule, length=3, radius=2, shapeName=dyn_cloth_B_04_capsule, rotation=<rot 90 0 0>}
 ]
 dyn_cloth_B_05 [
  {shapeType=Capsule, length=3, radius=2, shapeName=dyn_cloth_B_05_capsule, center=<0.1 0 0>, rotation=<rot 90 0 0>}
  {shapeType=Capsule, length=3, radius=2, shapeName=dyn_cloth_B_05_capsule, center=<3.5 0 0>, rotation=<rot 90 0 0>}
 ]
 dyn_cloth_base [{shapeType=Capsule, length=3, radius=2, shapeName=dyn_cloth_base_capsule, center=<0.1 0 0>, rotation=<rot 90 0 0>}]
 dyn_cloth_base_MSTR [{shapeType=Capsule, length=1, radius=1, shapeName=dyn_cloth_base_MSTR_capsule}]
 pelvis_BucketKick_Backpack [{shapeType=Capsule, length=8, radius=12, center=<1.1206 -4.5 -0.1981>}]
 spine_02_BucketKick_Backpack [
  {shapeType=Capsule, length=2.5219, radius=9.7317, center=<-2.6885 -1.2641 0>}
  {shapeType=Capsule, length=20.6875, radius=9.3125, center=<-1.8317 1.9362 0>, rotation=<rot 90 0 0>}
 ]
 spine_03_BucketKick_Backpack [
  {shapeType=Capsule, length=6.1863, radius=8.4854, center=<0.0574 0.4416 0>}
  {shapeType=Capsule, length=20.7188, radius=9.2812, center=<-1.2163 0.549 0>, rotation=<rot 90 180 -175.0732>}
 ]
 thigh_l_BucketKick_Backpack [
  {shapeType=Capsule, length=14, radius=11, shapeName=thigh_l_capsule, center=<-15 0 0>, rotation=<rot 90 89.4328 95>}
  {shapeType=Capsule, length=5, radius=10, shapeName=thigh_l_capsule}
 ]
 thigh_r_BucketKick_Backpack [
  {shapeType=Capsule, length=14, radius=11, shapeName=thigh_l_capsule, center=<15 0 0>, rotation=<rot 90 166.5268 -7.9059>}
  {shapeType=Capsule, length=5, radius=10, shapeName=thigh_r_capsule}
 ]
}
constraints [
 {bone1Name=dyn_cloth_base, bone2Name=dyn_cloth_base_MSTR, swing1LimitDegrees=30, swing2LimitDegrees=10, twistLimitDegrees=5}
 {bone1Name=dyn_cloth_A_01, bone2Name=dyn_cloth_base, swing1LimitDegrees=20, swing2LimitDegrees=15, twistLimitDegrees=5}
 {bone1Name=dyn_cloth_B_01, bone2Name=dyn_cloth_base, swing1LimitDegrees=20, swing2LimitDegrees=15, twistLimitDegrees=5}
 {bone1Name=dyn_cloth_A_02, bone2Name=dyn_cloth_A_01, swing1LimitDegrees=20, swing2LimitDegrees=15, twistLimitDegrees=10}
 {bone1Name=dyn_cloth_A_03, bone2Name=dyn_cloth_A_02, swing1LimitDegrees=15, swing2LimitDegrees=15, twistLimitDegrees=10}
 {bone1Name=dyn_cloth_B_02, bone2Name=dyn_cloth_B_01, swing1LimitDegrees=20, swing2LimitDegrees=15, twistLimitDegrees=10}
 {bone1Name=dyn_cloth_B_03, bone2Name=dyn_cloth_B_02, swing1LimitDegrees=20, swing2LimitDegrees=15, twistLimitDegrees=10}
 {bone1Name=dyn_cloth_B_04, bone2Name=dyn_cloth_B_03, swing1LimitDegrees=10, swing2LimitDegrees=10, twistLimitDegrees=10}
 {bone1Name=dyn_cloth_B_05, bone2Name=dyn_cloth_B_04, swing1LimitDegrees=10, swing2LimitDegrees=10, twistLimitDegrees=10}
]
