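# Animation Blueprint for the BucketKick backpack cosmetic, posing the pack and simulating its dangling cloth.
# Each update caches body-mesh socket transforms, spine-driven pack offsets, movement states and a gravity override.
# Bone modifiers for proxy body bones and pack offsets build a cached start pose.
# bRigidBodyEnabled selects between a RigidBody simulation and an AnimDynamics chain, both with front-end tweaks, the latter also adjusting for parachuting.
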AnimBlueprint /BRCosmetics/Accessories/FORT_Backpacks/Backpack_F_MED_BucketKick/Meshes/F_MED_BucketKick_Pack_AnimBP
target_skeleton=&/F_MED_BucketKick_Pack_Skeleton
graphs [
 AnimGraph {nodes [
   n1 = Anim Dynamics - Bone: dyn_cloth_base!(ComponentPose=Local To Component(LocalPose=Use cached pose 'Start Pose'()), Alpha=1, GravityOverride:Vector="0,0,0", ExternalForce="0,0,0") -> (Pose) {
    angularDampingOverride=0.99
    bGravityOverrideInSimSpace=true
    bOverrideAngularDamping=true
    bUseGravityOverride=true
    simulationSpace=Actor
    alphaBoolBlend {blendOption=Linear}
    boundBone {boneName=dyn_cloth_base}
    componentAppliedLinearAccClamp=<2000 2000 2000>
    componentLinearAccScale=<1 1 1>
    componentLinearVelScale=<1 1 1>
    physicsBodyDefinitions [{
      boundBone {boneName=dyn_cloth_base}
      boxExtents=<10 15 15>
      constraintSetup {angularLimitsMax=<5 15 5>, angularLimitsMin=<-5 -15 -40>, angularTarget=<0 0 0>}
      localJointOffset=<1.5 0 0>
     }]
   }
   n2 = Anim Dynamics - Bone: dyn_cloth_A_02!(ComponentPose=@n1, Alpha=1, GravityOverride:Vector="0,0,0", ExternalForce="0,0,0") -> (Pose) {
    angularDampingOverride=0.9
    bGravityOverrideInSimSpace=true
    bOverrideAngularDamping=true
    bUseGravityOverride=true
    simulationSpace=Actor
    alphaBoolBlend {blendOption=Linear}
    boundBone {boneName=dyn_cloth_A_02}
    componentAppliedLinearAccClamp=<1500 1500 1500>
    componentLinearAccScale=<1 1 1>
    componentLinearVelScale=<1 1 1>
    physicsBodyDefinitions [{
      boundBone {boneName=dyn_cloth_A_02}
      boxExtents=<8 15 15>
      constraintSetup {angularLimitsMax=<5 20 5>, angularLimitsMin=<-5 -20 -30>, angularTarget=<0 0 0>}
      localJointOffset=<1 0 0>
     }]
   }
   n3 = Anim Dynamics - Bone: dyn_cloth_B_02!(ComponentPose=@n2, Alpha=1, GravityOverride:Vector="0,0,0", ExternalForce="0,0,0") -> (Pose) {
    angularDampingOverride=0.99
    bGravityOverrideInSimSpace=true
    bOverrideAngularDamping=true
    bUseGravityOverride=true
    simulationSpace=Actor
    alphaBoolBlend {blendOption=Linear}
    boundBone {boneName=dyn_cloth_B_02}
    componentAppliedLinearAccClamp=<1000 1000 1000>
    componentLinearAccScale=<1 1 1>
    componentLinearVelScale=<1 1 1>
    physicsBodyDefinitions [{
      boundBone {boneName=dyn_cloth_B_02}
      boxExtents=<10 15 15>
      constraintSetup {angularLimitsMax=<5 20 5>, angularLimitsMin=<-5 -20 -30>, angularTarget=<0 0 0>}
      localJointOffset=<1.5 0 0>
     }]
   }
   n4 = Anim Dynamics - Bone: dyn_cloth_B_04!(ComponentPose=@n3, Alpha=1, GravityOverride:Vector="0,0,0", ExternalForce="0,0,0") -> (Pose) {
    angularDampingOverride=0.99
    bGravityOverrideInSimSpace=true
    bOverrideAngularDamping=true
    bUseGravityOverride=true
    simulationSpace=Actor
    alphaBoolBlend {blendOption=Linear}
    boundBone {boneName=dyn_cloth_B_04}
    componentAppliedLinearAccClamp=<500 500 500>
    componentLinearAccScale=<1 1 1>
    componentLinearVelScale=<1 1 1>
    physicsBodyDefinitions [{
      boundBone {boneName=dyn_cloth_B_04}
      boxExtents=<5 15 15>
      constraintSetup {angularLimitsMax=<5 25 5>, angularLimitsMin=<-5 -25 -25>, angularTarget=<0 0 0>}
      localJointOffset=<1.5 0 0>
     }]
   }
   n5 = Get bParachuteOpenedAndStandingUpright() -> (bParachuteOpenedAndStandingUpright:bool)
   n6 = "Transform (Modify) Bone - Bone: dyn_cloth_base"!(ComponentPose=@n4, bAlphaBoolEnabled:bool=@n5, Translation="0,0,0", Rotation="0,0,15", Scale="1,1,1") -> (Pose) {
    alphaInputType=Bool
    rotationMode=BMM_Additive
    translationMode=BMM_Additive
    alphaBoolBlend {blendInTime=0.2, blendOption=CubicInOut, blendOutTime=0.2}
    boneToModify {boneName=dyn_cloth_base}
    rotation {roll=15}
   }
   n7 = "Transform (Modify) Bone - Bone: pack"!(ComponentPose=@n6, bAlphaBoolEnabled:bool=@n5, Translation="0,0,0", Rotation="0,0,0", Scale) -> (Pose) {
    alphaInputType=Bool
    rotationMode=BMM_Additive
    alphaBoolBlend {blendInTime=0.2, blendOption=Linear, blendOutTime=0.2}
    boneToModify {boneName=pack}
   }
   n8 = "Transform (Modify) Bone - Bone: dyn_cloth_base"!(ComponentPose=@n7, bAlphaBoolEnabled=Get bIsInFrontEnd(), Rotation="0,0,10") -> (Pose) {
    alphaInputType=Bool
    rotationMode=BMM_Additive
    alphaBoolBlend {blendOption=Linear}
    boneToModify {boneName=dyn_cloth_base}
    rotation {roll=10}
   }
   n9 = RigidBody!(ComponentPose=Local To Component(LocalPose=Use cached pose 'Start Pose'()), Alpha, OverrideWorldGravity:Vector="0,0,0") -> (Pose) {
    bOverrideWorldGravity=true
    overridePhysicsAsset=&/F_MED_BucketKick_Pack_Physics
    alphaBoolBlend {blendOption=Linear}
    baseBoneRef {boneName=root}
    componentAppliedLinearAccClamp=<1000 1000 1000>
    componentLinearAccScale=<0.25 0.25 0.25>
    componentLinearVelScale=<0.3 0.3 0.3>
    simSpaceSettings {
     maxAngularAcceleration=150
     maxAngularVelocity=5
     maxLinearAcceleration=200
     maxLinearVelocity=200
     worldAlpha=1
     externalLinearDragV=<0.7 0.7 0.7>
    }
   }
   n10 = "Transform (Modify) Bone - Bone: dyn_cloth_B_04"!(ComponentPose=@n9, bAlphaBoolEnabled:bool=Get bIsInFrontEnd(), Rotation="0,0,10") -> (Pose) {
    alphaInputType=Bool
    rotationMode=BMM_Additive
    alphaBoolBlend {blendOption=Linear}
    boneToModify {boneName=dyn_cloth_B_04}
    rotation {roll=10}
   }
   Output Pose(Result=Blend Poses by bool(BlendPose_0=Component To Local(ComponentPose=@n10), BlendPose_1=Component To Local(ComponentPose=@n8), bActiveValue=Get bRigidBodyEnabled()) {blendType=Linear}) {layerGroup=None}
   n11 = "Transform (Modify) Bone - Bone: pack"!(ComponentPose=Local To Component(LocalPose=Input Pose()), Alpha:float, Translation="0,2,0", Rotation="0,0,5") -> (Pose) {
    rotationMode=BMM_Additive
    rotationSpace=BCS_BoneSpace
    translationMode=BMM_Additive
    alphaBoolBlend {blendOption=Linear}
    boneToModify {boneName=pack}
    rotation {roll=5}
    translation {y=2}
   }
   n12 = "Transform (Modify) Bone - Bone: pelvis_BucketKick_Backpack"!(ComponentPose=@n11, Alpha=1, Translation="0,0,0", Rotation="0,0,0") -> (Pose) {
    rotationMode=BMM_Replace
    rotationSpace=BCS_WorldSpace
    translationMode=BMM_Replace
    translationSpace=BCS_WorldSpace
    alphaBoolBlend {blendOption=Linear}
    boneToModify {boneName=pelvis_BucketKick_Backpack}
   }
   n13 = "Transform (Modify) Bone - Bone: spine_02_BucketKick_Backpack"!(ComponentPose=@n12, Alpha=1, Translation="0,0,0", Rotation="0,0,0") -> (Pose) {
    rotationMode=BMM_Replace
    rotationSpace=BCS_WorldSpace
    translationMode=BMM_Replace
    translationSpace=BCS_WorldSpace
    alphaBoolBlend {blendOption=Linear}
    boneToModify {boneName=spine_02_BucketKick_Backpack}
   }
   n14 = "Transform (Modify) Bone - Bone: spine_03_BucketKick_Backpack"!(ComponentPose=@n13, Alpha=1, Translation="0,0,0", Rotation="0,0,0") -> (Pose) {
    rotationMode=BMM_Replace
    rotationSpace=BCS_WorldSpace
    translationMode=BMM_Replace
    translationSpace=BCS_WorldSpace
    alphaBoolBlend {blendOption=Linear}
    boneToModify {boneName=spine_03_BucketKick_Backpack}
   }
   n15 = "Transform (Modify) Bone - Bone: thigh_l_BucketKick_Backpack"!(ComponentPose=@n14, Alpha=1, Translation="0,0,0", Rotation="0,0,0") -> (Pose) {
    rotationMode=BMM_Replace
    rotationSpace=BCS_WorldSpace
    translationMode=BMM_Replace
    translationSpace=BCS_WorldSpace
    alphaBoolBlend {blendOption=Linear}
    boneToModify {boneName=thigh_l_BucketKick_Backpack}
   }
   n16 = "Transform (Modify) Bone - Bone: thigh_r_BucketKick_Backpack"!(ComponentPose=@n15, Alpha=1, Translation="0,0,0", Rotation="0,0,0") -> (Pose) {
    rotationMode=BMM_Replace
    rotationSpace=BCS_WorldSpace
    translationMode=BMM_Replace
    translationSpace=BCS_WorldSpace
    alphaBoolBlend {blendOption=Linear}
    boneToModify {boneName=thigh_r_BucketKick_Backpack}
   }
   n17 = "Transform (Modify) Bone - Bone: pack"!(ComponentPose=@n16, Alpha, Translation="0,0,0", Rotation="0,0,0", Scale:Vector) -> (Pose) {
    rotationMode=BMM_Additive
    rotationSpace=BCS_BoneSpace
    translationMode=BMM_Additive
    translationSpace=BCS_BoneSpace
    alphaBoolBlend {blendOption=Linear}
    boneToModify {boneName=pack}
   }
   n18 = "Transform (Modify) Bone - Bone: pack"!(
    ComponentPose=@n17
    bAlphaBoolEnabled:bool=Get Is Sky Diving and NOT Parashuting() -> (Is Sky Diving and NOT Parashuting:bool)
    Translation="0,-3,0"
    Rotation="0,0,0"
   ) -> (Pose) {
    alphaInputType=Bool
    rotationMode=BMM_Additive
    rotationSpace=BCS_BoneSpace
    translationMode=BMM_Additive
    translationSpace=BCS_BoneSpace
    alphaBoolBlend {blendInTime=0.2, blendOption=CubicInOut, blendOutTime=0.2}
    boneToModify {boneName=pack}
    translation {y=-3}
   }
   n19 = "Transform (Modify) Bone - Bone: pack"(
    ComponentPose=@n18
    bAlphaBoolEnabled=Get Is Crouching and Moving 2D() -> (Is Crouching and Moving 2D:bool)
    Translation="0,0,0"
    Rotation="0,0,-10"
   ) {
    alphaInputType=Bool
    rotationMode=BMM_Additive
    rotationSpace=BCS_BoneSpace
    translationMode=BMM_Additive
    translationSpace=BCS_BoneSpace
    alphaBoolBlend {blendInTime=0.2, blendOption=CubicInOut, blendOutTime=0.2}
    boneToModify {boneName=pack}
    rotation {roll=-10}
   }
   Save cached pose 'Start Pose'(Pose=Component To Local(ComponentPose=@n19)) {cacheName=Start Pose}
  ]}
 EventGraph {nodes [
   Event BlueprintUpdateAnimation()
   > n1 = Gravity Override State Function DEPRECATED(CharacterPartAnimInstance=Self-Reference(), GameTimeInSeconds=GetGameTimeInSeconds(), GravityOverrideParams=Get Gravity Override Params())
   > Set Gravity Override(Gravity Override=@n1)
   n2 = Get BodyMesh() -> (BodyMesh:Skeletal Mesh Component Object Reference)
   n3 = BreakTransform(InTransform=GetSocketTransform(self=@n2, InSocketName=pelvis, TransformSpace=RTS_World))
   > Set Pelvis Translation(Pelvis Translation:Vector=@n3) -> (Output_Get:Vector)
   > Set Pelvis Rotation(Pelvis Rotation:Rotator=@n3.Rotation) -> (Output_Get:Rotator)
   n4 = BreakTransform(InTransform=GetSocketTransform(self=@n2, InSocketName=spine_02, TransformSpace=RTS_World))
   > Set Spine 02 Translation(Spine 02 Translation:Vector=@n4) -> (Output_Get:Vector)
   > Set Spine 02 Rotation(Spine 02 Rotation:Rotator=@n4.Rotation) -> (Output_Get:Rotator)
   n5 = BreakTransform(InTransform=GetSocketTransform(self=@n2, InSocketName=spine_03, TransformSpace=RTS_World))
   > Set Spine 03 Translation(Spine 03 Translation:Vector=@n5) -> (Output_Get:Vector)
   > Set Spine 03 Rotation(Spine 03 Rotation:Rotator=@n5.Rotation) -> (Output_Get:Rotator)
   n6 = BreakTransform(InTransform=GetSocketTransform(self=@n2, InSocketName=thigh_l, TransformSpace=RTS_World))
   > Set Thigh L Translation(Thigh L Translation:Vector=@n6) -> (Output_Get:Vector)
   > "Set Thigh  L Rotation"("Thigh  L Rotation":Rotator=@n6.Rotation) -> (Output_Get:Rotator)
   n7 = BreakTransform(InTransform=GetSocketTransform(self=@n2, InSocketName=thigh_r, TransformSpace=RTS_World))
   > Set Thigh R Translation(Thigh R Translation:Vector=@n7) -> (Output_Get:Vector)
   > "Set Thigh  R Rotation"("Thigh  R Rotation":Rotator=@n7.Rotation) -> (Output_Get:Rotator)
   n8 = GetOwnerSkeletalMeshComponent()
   n9 = BreakRotator(InRot=BreakTransform(InTransform=GetDeltaTransformFromRefPose(self=@n8, BoneName=spine_05, BaseName=spine_04)).Rotation)
   > Set BackPack Rotation(BackPack Rotation:Rotator=MakeRotator(Roll=MapRangeClamped(Value=@n9.Yaw, InRangeA=0, InRangeB=-30, OutRangeB=30))) -> (Output_Get:Rotator)
   > Set BackPack Translation(BackPack Translation:Vector=MakeVector(Y=MapRangeClamped(Value=@n9.Yaw, InRangeA=0, InRangeB=-30, OutRangeB=-2))) -> (Output_Get:Vector)
   n10 = GetOwnerSkeletalMeshComponent()
   n11 = BreakRotator(InRot=BreakTransform(InTransform=GetDeltaTransformFromRefPose(self=@n10, BoneName=pelvis, BaseName=spine_03)).Rotation)
   > Set BackPack Parashuting Rotation(BackPack Parashuting Rotation:Rotator=MakeRotator(Roll=MapRangeClamped(Value=@n11.Yaw, InRangeA=0, InRangeB=20, OutRangeB=20))) -> (Output_Get:Rotator)
   n12 = Get bParachuteOpenedAndStandingUpright()
   > Set Is Sky Diving and NOT Parashuting(Is Sky Diving and NOT Parashuting:bool=AND Boolean(A=Get bIsSkyDiving(), B=NOT Boolean(A=@n12))) -> (Output_Get:bool)
   > Set Is Crouching and Moving 2D(Is Crouching and Moving 2D:bool=AND Boolean(A=Get bIsCrouching(), B=Get bIsMoving2D())) -> (Output_Get:bool)
   n13 = Event OnCharacterCustomizationLoaded()
   > Set BodyMesh(BodyMesh=Get Array Item with EFort Custom Part Type(Array=@n13.CustomMeshParts))
   TryGetPawnOwner()
  ]}
]
variables [
 BackPack Parashuting Rotation {type=Rotator}
 BackPack Rotation {type=Rotator}
 BackPack Translation {type=Vector}
 BodyMesh {type=SkeletalMeshComponent}
 Gravity Override {type=Vector}
 Gravity Override Params {type=GravityOverrideParamsStruct}
 Is Crouching and Moving 2D {type=boolean}
 Is Sky Diving and NOT Parashuting {type=boolean}
 Pelvis Rotation {type=Rotator}
 Pelvis Translation {type=Vector}
 Spine 02 Rotation {type=Rotator}
 Spine 02 Translation {type=Vector}
 Spine 03 Rotation {type=Rotator}
 Spine 03 Translation {type=Vector}
 "Thigh  L Rotation" {type=Rotator}
 "Thigh  R Rotation" {type=Rotator}
 Thigh L Translation {type=Vector}
 Thigh R Translation {type=Vector}
]
